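; Blueprint that periodically triggers timed wind gusts through the global Fortnite material parameter collection, with a matching sound attached to player 0's pawn.
; BeginPlay arms a timer with random delay for StartWindGusts, which sets active flag, duration and timestamp, spawns the sound with a Duration parameter, then waits.
; It then clears the flag and re-arms the timer.
; DebugWindGust runs the same sequence immediately after shortening the delay range.
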(event EventBeginPlay
  (bind _returnvalue (Utilities|Time|SetTimerbyEvent (delegate Custom|StartWindGusts) (Math|Random|RandomFloatinRange (Variables|Default|GetDelayTime_Min) (Variables|Default|GetDelayTime_Max))))
  (Variables|Default|SetTimer _returnvalue))

(event Custom|DebugWindGust
  (Variables|Default|SetDelayTime_Min 2.0)
  (Variables|Default|SetDelayTime_Max 3.0)
  (Rendering|Material|SetScalarParameterValue "/Game/Packages/Fortress_SharedMaterials/GlobalMaterialParameters/FortniteMaterialParameters.FortniteMaterialParameters" "WindGustsSystem_Active" 1.0 :!class /Script/Engine.KismetMaterialLibrary)
  (Rendering|Material|SetScalarParameterValue "/Game/Packages/Fortress_SharedMaterials/GlobalMaterialParameters/FortniteMaterialParameters.FortniteMaterialParameters" "WindGustsSystem_Duration" (Variables|Default|GetDuration) :!class /Script/Engine.KismetMaterialLibrary)
  (Rendering|Material|SetScalarParameterValue "/Game/Packages/Fortress_SharedMaterials/GlobalMaterialParameters/FortniteMaterialParameters.FortniteMaterialParameters" "WindGustsSystem_TimeStamp" (Utilities|Time|GetGameTimeinSeconds) :!class /Script/Engine.KismetMaterialLibrary)
  (bind _rootcomponent (Variables|/Script/Engine.Actor|GetRootComponent (Game|GetPlayerPawn 0))
    (:then
      (bind _returnvalue (Audio|SpawnSoundAttached (Variables|Default|GetSound) _rootcomponent))
      (Audio|Parameter|SetFloatParameter _returnvalue "Duration" (Variables|Default|GetDuration) :!class /Script/AudioExtensions.AudioParameterControllerInterface))
    (:else
      _))
  (Utilities|FlowControl|Delay (Variables|Default|GetDuration))
  (Rendering|Material|SetScalarParameterValue "/Game/Packages/Fortress_SharedMaterials/GlobalMaterialParameters/FortniteMaterialParameters.FortniteMaterialParameters" "WindGustsSystem_Active" :!class /Script/Engine.KismetMaterialLibrary)
  (bind _returnvalue_2 (Utilities|Time|SetTimerbyEvent (delegate Custom|StartWindGusts) (Math|Random|RandomFloatinRange (Variables|Default|GetDelayTime_Min) (Variables|Default|GetDelayTime_Max))))
  (Variables|Default|SetTimer _returnvalue_2))

(event Custom|StartWindGusts
  (Rendering|Material|SetScalarParameterValue "/Game/Packages/Fortress_SharedMaterials/GlobalMaterialParameters/FortniteMaterialParameters.FortniteMaterialParameters" "WindGustsSystem_Active" 1.0 :!class /Script/Engine.KismetMaterialLibrary)
  (Rendering|Material|SetScalarParameterValue "/Game/Packages/Fortress_SharedMaterials/GlobalMaterialParameters/FortniteMaterialParameters.FortniteMaterialParameters" "WindGustsSystem_Duration" (Variables|Default|GetDuration) :!class /Script/Engine.KismetMaterialLibrary)
  (Rendering|Material|SetScalarParameterValue "/Game/Packages/Fortress_SharedMaterials/GlobalMaterialParameters/FortniteMaterialParameters.FortniteMaterialParameters" "WindGustsSystem_TimeStamp" (Utilities|Time|GetGameTimeinSeconds) :!class /Script/Engine.KismetMaterialLibrary)
  (bind _rootcomponent (Variables|/Script/Engine.Actor|GetRootComponent (Game|GetPlayerPawn 0))
    (:then
      (bind _returnvalue (Audio|SpawnSoundAttached (Variables|Default|GetSound) _rootcomponent))
      (Audio|Parameter|SetFloatParameter _returnvalue "Duration" (Variables|Default|GetDuration) :!class /Script/AudioExtensions.AudioParameterControllerInterface))
    (:else
      _))
  (Utilities|FlowControl|Delay (Variables|Default|GetDuration))
  (Rendering|Material|SetScalarParameterValue "/Game/Packages/Fortress_SharedMaterials/GlobalMaterialParameters/FortniteMaterialParameters.FortniteMaterialParameters" "WindGustsSystem_Active" :!class /Script/Engine.KismetMaterialLibrary)
  (bind _returnvalue_2 (Utilities|Time|SetTimerbyEvent (delegate Custom|StartWindGusts) (Math|Random|RandomFloatinRange (Variables|Default|GetDelayTime_Min) (Variables|Default|GetDelayTime_Max))))
  (Variables|Default|SetTimer _returnvalue_2))

(fn ConstructionScript)
