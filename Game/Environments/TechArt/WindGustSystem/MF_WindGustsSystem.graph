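# Material function that exposes timing and fade values for the wind gust system to materials.
# It reads Active, TimeStamp and Duration from FortniteMaterialParameters and compares Time against the stamp.
# Saturated divisions by adjustable fade times yield fade-in, fade-out, their product, normalized linear progression and a post-gust reset fade.
# The Active flag gates progression and blends the separate fade and reset outputs, while the combined fade product ignores it.
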MaterialFunction /Game/Environments/TechArt/WindGustSystem/MF_WindGustsSystem
graph {
 properties {description=FadeinFadeOutNormalized}
 nodes [
  n1 = CollectionParameter() {collection=&FortniteMaterialParameters, parameterName=WindGustsSystem_Active}
  FunctionOutput(None=NamedRerouteDeclaration(None=ComponentMask(None=@n1) {r=true}) {name=Active, nodeColor=<rgba 1 0.8438 0 1>}) {description="0 = off and 1 = on", outputName=Active}
  n2 = CollectionParameter() {
   collection=&FortniteMaterialParameters
   parameterName=WindGustsSystem_TimeStamp
  }
  n3 = ComponentMask(None=@n2) {r=true}
  FunctionOutput(None=@n3) {description=Gametime Stamp, outputName=TimeStamp, sortPriority=2}
  n4 = CollectionParameter() {
   collection=&FortniteMaterialParameters
   parameterName=WindGustsSystem_Duration
  }
  n5 = ComponentMask(None=@n4) {r=true}
  FunctionOutput(None=@n5) {description=Windgust Duration in seconds, outputName=Duration, sortPriority=3}
  n6 = Subtract(A=Time(), B=@n3)
  n7 = Constant() {r=2}
  n8 = Saturate(None=Divide(A=@n6, B=FunctionInput(Preview=@n7) {bUsePreviewValueAsDefault=true, inputName=FadeInTime, inputType=FunctionInput_Scalar}))
  n9 = Subtract(A=@n6, B=@n5)
  n10 = Constant() {r=2}
  n11 = Saturate(None=Divide(
    A=Multiply(A=@n9) {constB=-1}
    B=FunctionInput(Preview=@n10) {bUsePreviewValueAsDefault=true, inputName=FadeOutTIme, inputType=FunctionInput_Scalar}
   ))
  FunctionOutput(None=Multiply(A=@n8, B=@n11)) {description=Gametime Stamp, outputName=FadeInFadeOutNormalized, sortPriority=1}
  n12 = Divide(A=@n6, B=@n5)
  FunctionOutput(None=Multiply(A=Saturate(None=@n12), B=NamedRerouteUsage())) {description=Windgust Duration in seconds, outputName=LinearTimeProgressionNormalized, sortPriority=10}
  n13 = Divide(A=@n9, B=FunctionInput(Preview=Constant() {r=2}) {bUsePreviewValueAsDefault=true, inputName=ResetFadeInTime, inputType=FunctionInput_Scalar})
  FunctionOutput(None=LinearInterpolate(A=Saturate(None=@n13), Alpha=NamedRerouteUsage())) {description=Windgust Duration in seconds, outputName=ResetFadeIn, sortPriority=10}
  n14 = NamedRerouteUsage()
  FunctionOutput(None=LinearInterpolate(B=@n8, Alpha=@n14) {constA=1}) {description=Gametime Stamp, outputName=FadeInOnly, sortPriority=11}
  FunctionOutput(None=LinearInterpolate(B=@n11, Alpha=@n14) {constA=1}) {description=Gametime Stamp, outputName=FadeOutOnly, sortPriority=11}
 ]
}
